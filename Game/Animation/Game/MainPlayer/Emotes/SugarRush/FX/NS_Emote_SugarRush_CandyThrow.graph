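# Niagara effect for the SugarRush player emote, showing a cluster of candy, likely held in the hand, and candy being thrown.
# CandyThrown launches two tinted mesh variants in a cone at a User.SpawnRateMultiplier-scaled, age-gated rate; they tumble, bounce under gravity and drag, then shrink.
# HandCandy arranges staggered bursts in rings on a sphere, then fakes respawning by cyclically scaling persistent particles and swapping meshes on randomly offset timers.
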NiagaraSystem /Game/Animation/Game/MainPlayer/Emotes/SugarRush/FX/NS_Emote_SugarRush_CandyThrow
dependencies {
 usedDataInterfaces [&NiagaraDataInterfaceCurve, &NiagaraDataInterfaceArrayFloat]
 usedDynamicInputs []
 usedModules [&SystemState]
}
emitters {
 CandyThrown {
  emitterUpdateScript [
   &EmitterState()
   &SpawnRate(
    Spawn Probability=&MakeCustomFloatFromBool(Bool=&SetBoolByFloatComparison(A=$Emitter.Age, B=0.5), True Float=1, False Float=0)
    SpawnRate=&Multiply_Float(A=10, B=$User.SpawnRateMultiplier)
    Use Spawn Probability=true
   )
  ]
  particleSpawnScript [
   &InitializeParticle(
    Color Mode=::NewEnumerator0
    Lifetime=&RandomRangeFloat(Minimum=1.2, Maximum=1.75)
    Mesh Scale Mode=::NewEnumerator3
    UsePositionOffset=false
    Write Material Random=true
   )
   &ShapeLocation(MaskSpawnGroup=false, Override Seed=false, Sphere Radius=3.5, UseOwnerScale=true)
   &AddVelocity(
    Cone Angle=25
    Cone Axis=<1 -1 0.75>
    Module.UseConeFalloff=false
    Module.UseVelDistribution=false
    Override Seed=false
    Velocity Mode=::NewEnumerator2
    Velocity Speed=&RandomRangeFloat(Minimum=350, Maximum=450)
   )
   /SetVariables(Particles.MeshIndex=&RandomRangeInt(Minimum=0, Maximum=1)) {isSetParameters=true, module=SetVariables_#1}
   &InitialMeshOrientation(Use Rotation Vector=false)
   &Color(Color=&LerpLinearColors(
     StartColor=&LerpLinearColors(StartColor=<rgba 1 1 1 1>, EndColor=<rgba 1 0.5 0 1>, LerpFactor=&UniformAOrBFloat(A=0))
     EndColor=&LerpLinearColors(StartColor=<rgba 1 1 1 1>, EndColor=<rgba 1 0.5 0 1>, LerpFactor=&UniformAOrBFloat(A=0))
     LerpFactor=&MakeFloatFromInt(Int=$Particles.MeshIndex)
    ))
  ]
  particleUpdateScript [
   &ParticleState()
   &UpdateMeshOrientation(
    Rotation Rate=&RandomRangeFloat(Minimum=0.5, Maximum=1.5)
    Rotation Vector=&RandomVector(Vector Scale=1)
   )
   &GravityForce()
   &Drag(Drag=&RandomRangeFloat(Minimum=1, Maximum=1.5), Use Linear Drag=true, Use Rotational Drag=true)
   &Collision(
    Age Colliding Particles=false
    Collision Enabled=&SetBoolByFloatComparison(A=$Particles.Age, B=0.5)
    Friction=1.2
    KillOnCollision=false
    Module.Control Rest State Via Surface Normal=false
    Module.EnableMaxCollisionCount=false
    Module.Randomize Collision Normal Vector=false
    Restitution=&RandomRangeFloat(Minimum=0.25, Maximum=0.35)
   )
   &ScaleMeshSize(Scale Factor=&VectorFromFloat(Value=&FloatFromCurve(FloatCurve=DataInterface() {propertyValues {Curve {keys [{interpMode=RCIM_Cubic, tangentMode=RCTM_User, time=0.8, value=1}, {interpMode=RCIM_Cubic, tangentMode=RCTM_User, time=1, value=0}]}}})))
   &SolveForcesAndVelocity(Clamp Velocity=false, Limit Acceleration=false)
  ]
  rendererClasses [&NiagaraMeshRendererProperties]
  renderers [{rendererClass=&NiagaraMeshRendererProperties, rendererIndex=0}]
 }
 HandCandy {
  emitterUpdateScript [
   &EmitterState(
    Inactive Response=::NewEnumerator1
    Life Cycle Mode=::NewEnumerator1
    Loop Behavior=::NewEnumerator1
    Loop Duration Mode=::NewEnumerator1
    UseLoopDelay=false
   )
   &SpawnBurst_Instantaneous(Spawn Count=5, Use Spawn Probability=false, UseLoopCountLimit=false)
   &SpawnBurst_Instantaneous(Spawn Count=4, Spawn Group=1, Spawn Time=0.1, Use Spawn Probability=false, UseLoopCountLimit=false) {module=SpawnBurst_Instantaneous001}
   &SpawnBurst_Instantaneous(Spawn Count=4, Spawn Group=2, Spawn Time=0.2, Use Spawn Probability=false, UseLoopCountLimit=false) {module=SpawnBurst_Instantaneous002}
   &SpawnBurst_Instantaneous(Spawn Count=2, Spawn Group=3, Spawn Time=0.15, Use Spawn Probability=false, UseLoopCountLimit=false) {module=SpawnBurst_Instantaneous003}
  ]
  particleSpawnScript [
   &InitializeParticle(Color Mode=::NewEnumerator0, Lifetime=5, Mesh Scale Mode=::NewEnumerator3, UsePositionOffset=false, Write Material Random=true)
   &ShapeLocation(
    MaskSpawnGroup=false
    Sphere Distribution=::NewEnumerator1
    Sphere Radius=2.5
    U Position=&ReturnNormalizedExecIndex()
    UseOwnerScale=true
    V Position=&SelectFloatFromArray(
     Direct Array Index=&SpawnGroup()
     Float Selection Array=DataInterface() {propertyValues {FloatData [0.5,0.65,0.8,0.95]}}
    )
   )
   /SetVariables(Particles.MeshIndex=&RandomRangeInt(Minimum=0, Maximum=1)) {isSetParameters=true, module=SetVariables_#1}
   &InitialMeshOrientation(Use Rotation Vector=false)
   &PartitionParticles(Number Of Partitions=4)
   /SetVariables(Particles.RefreshRate=1) {isSetParameters=true}
   /SetVariables(Particles.TimeDelay:NiagaraFloat=&RandomRangeFloat(Minimum=0, Maximum=1)) {isSetParameters=true}
   &Color(Color=&LerpLinearColors(
     StartColor=&LerpLinearColors(StartColor=<rgba 1 1 1 1>, EndColor=<rgba 1 0.5 0 1>, LerpFactor=&UniformAOrBFloat(A=0))
     EndColor=&LerpLinearColors(StartColor=<rgba 1 1 1 1>, EndColor=<rgba 1 0.5 0 1>, LerpFactor=&UniformAOrBFloat(A=0))
     LerpFactor=&MakeFloatFromInt(Int=$Particles.MeshIndex)
    ))
  ]
  particleUpdateScript [
   &ParticleState(Kill Particles When Lifetime Has Elapsed=false)
   &ScaleMeshSize(Scale Factor=&VectorFromFloat(Value=&FloatFromCurve(
      FloatCurve=DataInterface() {propertyValues {Curve {keys [{interpMode=RCIM_Cubic, tangentMode=RCTM_User, time=0, value=0}, {interpMode=RCIM_Cubic, tangentMode=RCTM_User, time=0.25, value=1}]}}}
      CurveIndex=&Add_Float(A=$Particles.Age, B=&RandomRangeFloat(Minimum=0, Maximum=-0.1))
      Scale Curve=&FloatFromCurve(
       FloatCurve=DataInterface() {propertyValues {Curve {keys [
           {interpMode=RCIM_Cubic, tangentMode=RCTM_User, time=0, value=0}
           {interpMode=RCIM_Cubic, tangentMode=RCTM_User, time=0.2, value=1}
           {interpMode=RCIM_Cubic, tangentMode=RCTM_User, time=0.8, value=1}
           {interpMode=RCIM_Cubic, tangentMode=RCTM_User, time=1, value=0}
          ]}}}
       CurveIndex=&Frac_Float(Float=&Divide_Float(A=&Add_Float(A=$Emitter.Age, B=$Particles.TimeDelay), B=$Particles.RefreshRate))
      )
     )))
   /SetVariables(Particles.MeshIndex=&ModuloInt(
     A=&Add_Integer(
      A=&RandomRangeInt(Minimum=0, Maximum=1)
      B=&MakeIntFromFloat(Float=/Niagara/DynamicInputs/Math/Floor_Float(Float:NiagaraFloat=&Divide_Float(A=&Add_Float(A=$Emitter.Age, B=$Particles.TimeDelay), B=$Particles.RefreshRate)))
     )
     B=2
    )) {isSetParameters=true}
   &SolveForcesAndVelocity(Clamp Velocity=false, Limit Acceleration=false)
  ]
  rendererClasses [&NiagaraMeshRendererProperties]
  renderers [{rendererClass=&NiagaraMeshRendererProperties, rendererIndex=0}]
 }
}
summary {
 systemName=NS_Emote_SugarRush_CandyThrow
 emitters [
  {emitterName=CandyThrown, rendererClasses [&NiagaraMeshRendererProperties]}
  {emitterName=HandCandy, rendererClasses [&NiagaraMeshRendererProperties]}
 ]
 userVariables [User.SpawnRateMultiplier {type=NiagaraFloat, defaultValue=1, description=""}]
}
